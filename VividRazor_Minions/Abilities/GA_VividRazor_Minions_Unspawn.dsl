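; Gameplay ability that despawns a VividRazor minion AI after it stays Unaware for a tunable time.
; It caches the pawn and bot controller, then listens for alert-level changes, starting a timer on Unaware and clearing it on Suspicious or Threatened.
; On expiry, it tags the pawn as about to derezz, updates its rezz-out gameplay cue, pushes behavior control, then force-kills it as self-inflicted after a short delay.
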(event Ability|EventActivateAbility
  (CallFunction|InitVariables)
  (AlertLevel|BindEventtoOnAlertLevelChangedEvent (Variables|InternalData|GetAIBotController) (EventDispatchers|CreateEvent :!event Custom|AIAlertLevelChanged)))

(event Ability|EventOnEndAbility
  (params (bWasCancelled :type Boolean))
  (bind _aibotcontroller (Variables|InternalData|GetAIBotController))
  (Utilities|IsValid _aibotcontroller
    (:Is%20Valid
      (AlertLevel|UnbindEventfromOnAlertLevelChangedEvent _aibotcontroller (EventDispatchers|CreateEvent :!event Custom|AIAlertLevelChanged)))))

(event Custom|AIAlertLevelChanged
  (params (BotController :type Fort%20Athena%20AIBot%20Controller%20Object%20Reference) (OldAlertLevel :type EAlertLevel%20Enum :default "Unaware") (NewAlertLevel :type EAlertLevel%20Enum :default "Unaware"))
  ; Check when AI goes into Alert or Unaware State
  (if (Utilities|Enum|Equal(Enum) NewAlertLevel "Unaware")
    (bind _returnvalue (Utilities|Time|SetTimerbyEvent (EventDispatchers|CreateEvent :!event Custom|DespawnAI) (ScalableFloat|GetValueatLevel (Variables|Default|GetUnawareTimeBeforeDespawn))))
    (Variables|InternalData|SetHNDL_DespawnTimer _returnvalue)
    (elif (or (Utilities|Enum|Equal(Enum) NewAlertLevel "Suspicious") (Utilities|Enum|Equal(Enum) NewAlertLevel "Threatened"))
      (Utilities|Time|ClearandInvalidateTimerbyHandle (Variables|InternalData|GetHNDL_DespawnTimer)))))

(event Custom|DespawnAI
  (bind _aipawn (Variables|InternalData|GetAIPawn))
  (bind _asbp_player_pawn_vivid_razor_minions_base (Utilities|Casting|CastToBP_PlayerPawn_VividRazor_Minions_Base _aipawn))
  (Ability|GameplayEffect|AddLooseGameplayTags _aipawn (Variables|Default|GetTag_AboutToDerezz))
  (Class|BPPlayerPawnVividRazorMinionsBase|UpdateRezzoutGC _asbp_player_pawn_vivid_razor_minions_base (Variables|Default|GetRezz%20Out%20GC))
  (AI|PushBehaviorControl (Variables|InternalData|GetAIBotController) -2097152)
  (Utilities|FlowControl|Delay 1.25)
  (FortPawn|ForceKill _asbp_player_pawn_vivid_razor_minions_base "(TagName=\"DeathCause.SelfInflicted\")" 0 0)
  (AI|PopBehaviorControl (Variables|InternalData|GetAIBotController)))

(fn InitVariables
  (bind _outfortplayerpawn (CallFunction|GetAvatarasValidFortPlayerPawn)
    (:ValidFortPlayerPawn
      (bind _output_get (Variables|InternalData|SetAIPawn _outfortplayerpawn))
      (bind _asfort_athena_aibot_controller (Utilities|Casting|CastToFortAthenaAIBotController (Pawn|GetController _output_get))
        (:then
          (Variables|InternalData|SetAIBotController _asfort_athena_aibot_controller)
          (return))
        (:CastFailed
          _)))
    (:AvatarCastFailed
      _))
  (Ability|EndAbility))
